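# Physics asset for the Birthday back bling (M_MED_FN_Birthday_Pack), giving the backpack mesh its collision and joint setup.
# A large capsule covers the pack bone and a smaller capsule covers dyn_child_01, which appears to be a dangling secondary part.
# A single constraint joins dyn_child_01 to the pack, locking the first swing axis and setting small angle values for the second swing and twist.
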PhysicsAsset /BRCosmetics/Accessories/FORT_Backpacks/Backpack_M_MED_FN_Birthday/Meshes/M_MED_FN_Birthday_Pack_Physics
bodies {
 dyn_child_01 [{shapeType=Capsule, length=13.8806, radius=5.5489, shapeName=dyn_child_01_capsule, center=<0.1 1.3033 5.3305>, rotation=<rot 0 0 90>}]
 pack [{shapeType=Capsule, length=17.6672, radius=17.6685, center=<0 -11.3177 1.8755>}]
}
constraints [{bone1Name=dyn_child_01, bone2Name=pack, swing1LimitDegrees=45, swing1Motion=Locked, swing2LimitDegrees=5, twistLimitDegrees=10}]
